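# Material function that adds view-dependent sparkles to the Strategic Spur glider's emissive, tinted by base color.
# A sparkle normal texture is rotated around an axis from world down and the vertex normal, with the angle animated by a UV-and-time sine wave.
# Its camera-facing response is sharpened, masked by a selectable TechArtMask channel, and scaled by near/far depth-fade intensity, reduced at evening and night.
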MaterialFunction /BRCosmetics/Gadgets/Assets/VinderTech_GliderChute/Glider_StrategicSpur/Materials/MF_Glider_SparklesView
graph {nodes [
  n1 = GetMaterialAttributes(None=NamedRerouteUsage()) {attributeGetTypes [#1, #2]}
  n2 = NamedRerouteUsage()
  n3 = Transform(None=Constant3Vector() {constant=<rgba 0 0 -1 0>}) {transformSourceType=TRANSFORMSOURCE_World, transformType=TRANSFORM_Tangent}
  n4 = VertexNormalWS()
  n5 = RotateAboutAxis(
   NormalizedRotationAxis=CrossProduct(A=Normalize(VectorInput=@n3), B=Transform(None=@n4) {transformSourceType=TRANSFORMSOURCE_World, transformType=TRANSFORM_Tangent})
   RotationAngle=NamedRerouteUsage()
   PivotPoint=Constant()
   Position=@n2
  )
  n6 = DotProduct(A=CameraVectorWS(), B=Transform(None=Add(A=@n2, B=@n5)))
  n7 = SmoothStep(Value=Power(Base=OneMinus(None=@n6))) {constMin=0.1}
  n8 = Power(Base=NamedRerouteUsage(), Exp=ScalarParameter() {defaultValue=2, parameterName=TechArtMaskPower}) {constExponent=3}
  n9 = LinearInterpolate(
   A=ScalarParameter() {defaultValue=2, group=Sparkles, parameterName=SparkleIntensity_Near}
   B=ScalarParameter() {defaultValue=4, group=Sparkles, parameterName=SparkleIntensity_Far}
   Alpha=MaterialFunctionCall(Fade Length=ScalarParameter() {defaultValue=200, parameterName=DepthLength}) {materialFunction=&CameraDepthFade}.Result
  )
  n10 = Divide(A=@n9) {constB=4}
  n11 = Multiply(
   A=Multiply(A=@n1.BaseColor, B=Multiply(A=@n7, B=@n8))
   B=MaterialFunctionCall(Value_Morning=@n9, Value_Day=@n9, Value_Evening=@n10, Value_Night=@n10) {materialFunction=&Scalar_Time_of_Day}.Result
  )
  FunctionOutput(None=SetMaterialAttributes!(MaterialAttributes=@n1.MaterialAttributes, EmissiveColor=Add(A=@n1.EmissiveColor, B=@n11)) {attributeSetTypes [#2]})
  n12 = ScalarParameter() {defaultValue=0.2, parameterName=Rotate_Radius}
  Multiply(A=@n12) {constB=-1}
  n13 = TextureSampleParameter2D() {
   group=Sparkles
   parameterName=Sparkle_Dr
   samplerType=SAMPLERTYPE_Normal
   texture=/BRCosmetics/Gadgets/Assets/VinderTech_GliderChute/Glider_StrategicSpur/Textures/T_StrategicSpur_HGlider_Tarp_Sparkle_Dr
  }
  NamedRerouteDeclaration(None=MaterialFunctionCall!(XY vector=ComponentMask(None=@n13.RGB) {g=true, r=true}, Z sign, Default Normalization vector) {materialFunction=/Engine/Functions/Engine_MaterialFunctions02/Math/DeriveNormalZ_Function}.Result) {name=SparkleNormal, nodeColor=<rgba 0.1719 1 0 1>}
  NamedRerouteDeclaration(None=FunctionInput() {inputName=MA, inputType=FunctionInput_MaterialAttributes}) {name=MA_IN, nodeColor=<rgba 1 0 0.4922 1>}
  NamedRerouteDeclaration(None=ChannelMaskParameter(None=FunctionInput() {inputName=TechArtMask, sortPriority=1, previewValue=<1 1 1 1>}) {parameterName=UseSparkleChannel}.R) {name=TechArtMask, nodeColor=<rgba 0 0.8125 1 1>}
  n14 = Multiply(A=TextureCoordinate(), B=ScalarParameter() {defaultValue=2, group=Sparkles, parameterName=Tilling})
  n15 = Time()
  n16 = ScalarParameter() {defaultValue=-0.15, group=Sparkles, parameterName=MovingSpeed}
  n17 = Sine(None=Add(A=ComponentMask(None=@n14) {g=true}, B=Multiply(A=@n15, B=@n16)))
  NamedRerouteDeclaration(None=LinearInterpolate(A=Constant(), B=@n12, Alpha=Saturate(None=@n17)) {constB=0.05}) {name=NormalRotateAngle, nodeColor=<rgba 0 1 0.9297 1>}
 ]}
